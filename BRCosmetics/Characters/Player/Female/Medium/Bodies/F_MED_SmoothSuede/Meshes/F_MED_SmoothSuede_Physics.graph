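# Physics asset for the F_MED_SmoothSuede female player body, set up for secondary motion of two belts and a hanging mask piece.
# Each belt has front and back capsule chains from a pelvis-linked master, closed into a loop at an anchor chain.
# Cross constraints tie the belts together, one angularly locked, and the mask box hangs off the main belt's back chain.
# Capsules cover pelvis, lower spine, both thighs, and only the left calf and foot.
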PhysicsAsset /BRCosmetics/Characters/Player/Female/Medium/Bodies/F_MED_SmoothSuede/Meshes/F_MED_SmoothSuede_Physics
bodies {
 calf_l [{shapeType=Capsule, length=27.31, radius=7.9028, shapeName=calf_l_capsule, center=<-23.6405 -2.7775 0>, rotation=<rot 90 0 0>}]
 dyn_belt_anchor_01 [{shapeType=Capsule, length=8, radius=3, shapeName=dyn_belt_anchor_01_capsule, center=<5.473 0 0>, rotation=<rot 90 0 0>}]
 dyn_belt_anchor_02 [{shapeType=Capsule, length=8, radius=3, shapeName=dyn_belt_anchor_02_capsule, rotation=<rot 0 0 90>}]
 dyn_belt_anchor_base [{shapeType=Capsule, length=0.505, radius=0.505, shapeName=dyn_belt_anchor_base_capsule, rotation=<rot 0 -90 0>}]
 dyn_belt_bk_01 [{shapeType=Capsule, length=6, radius=2.5, shapeName=dyn_belt_bk_01_capsule, center=<4.1566 0 0>, rotation=<rot 90 0 0>}]
 dyn_belt_bk_02 [{shapeType=Capsule, length=6, radius=2.5, shapeName=dyn_belt_bk_02_capsule, center=<4.1566 0 0>, rotation=<rot 90 0 0>}]
 dyn_belt_bk_03 [{shapeType=Capsule, length=6, radius=2.5, shapeName=dyn_belt_bk_03_capsule, center=<4.1566 0 0>, rotation=<rot 90 0 0>}]
 dyn_belt_bk_04 [{shapeType=Capsule, length=6, radius=2.5, shapeName=dyn_belt_bk_04_capsule, center=<4.1566 0 0>, rotation=<rot 90 0 0>}]
 dyn_belt_bk_05 [{shapeType=Capsule, length=6, radius=2.5, shapeName=dyn_belt_bk_05_capsule, center=<4.1566 0 0>, rotation=<rot 90 0 0>}]
 dyn_belt_bk_06 [{shapeType=Capsule, length=6, radius=2.5, shapeName=dyn_belt_bk_06_capsule, center=<4.1566 0 0>, rotation=<rot 90 0 0>}]
 dyn_belt_bk_07 [{shapeType=Capsule, length=6, radius=2.5, shapeName=dyn_belt_bk_07_capsule, center=<4.1566 0 0>, rotation=<rot 90 0 0>}]
 dyn_belt_frnt_01 [{shapeType=Capsule, length=6, radius=2.5, shapeName=dyn_belt_frnt_01_capsule, center=<4.1554 0 0>, rotation=<rot 90 0 0>}]
 dyn_belt_frnt_02 [{shapeType=Capsule, length=5.0548, radius=3.4452, shapeName=dyn_belt_frnt_02_capsule, center=<4.1554 0 0>, rotation=<rot 90 0 0>}]
 dyn_belt_frnt_03 [{shapeType=Capsule, length=6, radius=2.5, shapeName=dyn_belt_frnt_03_capsule, center=<4.1554 0 0>, rotation=<rot 90 0 0>}]
 dyn_belt_frnt_04 [{shapeType=Capsule, length=5.5081, radius=2.9919, shapeName=dyn_belt_frnt_04_capsule, center=<4.1554 0 0>, rotation=<rot 90 0 0>}]
 dyn_belt_frnt_05 [{shapeType=Capsule, length=5.1032, radius=3.3968, shapeName=dyn_belt_frnt_05_capsule, center=<4.1554 0 0>, rotation=<rot 90 0 0>}]
 dyn_belt_frnt_06 [{shapeType=Capsule, length=5.6245, radius=2.8755, shapeName=dyn_belt_frnt_06_capsule, center=<4.1554 0 0>, rotation=<rot 90 0 0>}]
 dyn_belt_frnt_07 [{shapeType=Capsule, length=6, radius=2.5, shapeName=dyn_belt_frnt_07_capsule, center=<4.1554 0 0>, rotation=<rot 90 0 0>}]
 dyn_belt_mask_01 [{
   shapeType=Box
   extentX=18.9537
   extentY=4.0395
   extentZ=22.8436
   shapeName=dyn_belt_mask_01_box
   center=<11.3572 5.2594 -0.3247>
   rotation=<rot 0 -93.305 -76.5538>
  }]
 dyn_belt_mask_base [{shapeType=Capsule, length=0.505, radius=0.505, shapeName=dyn_belt_mask_base_capsule, rotation=<rot 0 -90 0>}]
 dyn_belt_master [{shapeType=Capsule, length=0, radius=0, shapeName=dyn_belt_master_capsule}]
 dyn_greenBelt_anchor_01 [{shapeType=Capsule, length=6, radius=2, shapeName=dyn_greenBelt_anchor_01_capsule, rotation=<rot 90 0 0>}]
 dyn_greenBelt_anchor_02 [{shapeType=Capsule, length=6, radius=2, shapeName=dyn_greenBelt_anchor_02_capsule, rotation=<rot 90 0 90>}]
 dyn_greenBelt_anchor_base [{shapeType=Capsule, length=6, radius=2, shapeName=dyn_greenBelt_anchor_base_capsule, rotation=<rot 90 0 0>}]
 dyn_greenBelt_bk_01 [{shapeType=Capsule, length=6, radius=2, shapeName=dyn_greenBelt_bk_01_capsule, center=<3.2751 0 0>, rotation=<rot 90 0 0>}]
 dyn_greenBelt_bk_02 [{shapeType=Capsule, length=6, radius=2, shapeName=dyn_greenBelt_bk_02_capsule, center=<3.2751 0 0>, rotation=<rot 90 0 0>}]
 dyn_greenBelt_bk_03 [{shapeType=Capsule, length=6, radius=2, shapeName=dyn_greenBelt_bk_03_capsule, center=<3.2751 0 0>, rotation=<rot 90 0 0>}]
 dyn_greenBelt_bk_04 [{shapeType=Capsule, length=6, radius=2, shapeName=dyn_greenBelt_bk_04_capsule, center=<3.2751 0 0>, rotation=<rot 90 0 0>}]
 dyn_greenBelt_bk_05 [{shapeType=Capsule, length=6, radius=2, shapeName=dyn_greenBelt_bk_05_capsule, center=<3.2751 0 0>, rotation=<rot 90 0 0>}]
 dyn_greenBelt_bk_06 [{shapeType=Capsule, length=6, radius=2, shapeName=dyn_greenBelt_bk_06_capsule, center=<3.2751 0 0>, rotation=<rot 90 0 0>}]
 dyn_greenBelt_bk_07 [{shapeType=Capsule, length=6, radius=2, shapeName=dyn_greenBelt_bk_07_capsule, center=<3.2751 0 0>, rotation=<rot 90 0 0>}]
 dyn_greenBelt_frnt_01 [{shapeType=Capsule, length=6, radius=2, shapeName=dyn_greenBelt_frnt_01_capsule, center=<3.9918 0 0>, rotation=<rot 90 0 0>}]
 dyn_greenBelt_frnt_02 [{shapeType=Capsule, length=6, radius=2, shapeName=dyn_greenBelt_frnt_02_capsule, center=<3.9918 0 0>, rotation=<rot 90 0 0>}]
 dyn_greenBelt_frnt_03 [{shapeType=Capsule, length=5.8115, radius=2.1885, shapeName=dyn_greenBelt_frnt_03_capsule, center=<3.9918 0 0>, rotation=<rot 90 0 0>}]
 dyn_greenBelt_frnt_04 [{shapeType=Capsule, length=5.709, radius=2.291, shapeName=dyn_greenBelt_frnt_04_capsule, center=<3.9918 0 0>, rotation=<rot 90 0 0>}]
 dyn_greenBelt_frnt_05 [{shapeType=Capsule, length=6, radius=2, shapeName=dyn_greenBelt_frnt_05_capsule, center=<3.9918 0 0>, rotation=<rot 90 0 0>}]
 dyn_greenBelt_frnt_06 [{shapeType=Capsule, length=6, radius=2, shapeName=dyn_greenBelt_frnt_06_capsule, center=<3.9918 0 0>, rotation=<rot 90 0 0>}]
 dyn_greenBelt_master [{shapeType=Capsule, length=6, radius=2, shapeName=dyn_greenBelt_master_capsule, rotation=<rot 90 0 0>}]
 foot_l [{shapeType=Capsule, length=11.78, radius=9.48, shapeName=foot_l_capsule, center=<-2.1087 5.5317 0>, rotation=<rot 90 -165.9638 103.6636>}]
 pelvis [{shapeType=Capsule, length=10.8137, radius=12.5876, shapeName=pelvis_capsule, center=<-7.7368 -3.2553 0.4227>}]
 spine_01 [{shapeType=Capsule, length=9.5908, radius=10.0414, shapeName=spine_01_capsule, center=<0 0.1008 0>}]
 spine_02 [{shapeType=Capsule, length=9.3704, radius=9.4801, shapeName=spine_02_capsule}]
 thigh_l [{shapeType=Capsule, length=44.9429, radius=10, shapeName=thigh_l_capsule, center=<-20.3593 -2.4307 0>, rotation=<rot 90 179.7016 -177.4779>}]
 thigh_r [{shapeType=Capsule, length=44.6615, radius=10, shapeName=thigh_r_capsule, center=<19.5757 1.3578 -0.0707>, rotation=<rot 90 180 -176.2765>}]
}
constraints [
 {
  bone1Name=dyn_belt_master
  bone2Name=pelvis
  swing1LimitDegrees=45
  swing1Motion=Free
  swing2LimitDegrees=45
  swing2Motion=Free
  twistLimitDegrees=45
  twistMotion=Free
 }
 {
  bone1Name=dyn_belt_bk_01
  bone2Name=dyn_belt_master
  swing1LimitDegrees=45
  swing1Motion=Free
  swing2LimitDegrees=45
  swing2Motion=Free
  twistLimitDegrees=45
  twistMotion=Free
 }
 {
  bone1Name=dyn_belt_bk_02
  bone2Name=dyn_belt_bk_01
  swing1LimitDegrees=45
  swing1Motion=Free
  swing2LimitDegrees=45
  swing2Motion=Free
  twistLimitDegrees=45
  twistMotion=Free
 }
 {bone1Name=dyn_belt_bk_03, bone2Name=dyn_belt_bk_02, swing1LimitDegrees=10, swing2LimitDegrees=10, twistLimitDegrees=5}
 {bone1Name=dyn_belt_bk_04, bone2Name=dyn_belt_bk_03, swing1LimitDegrees=10, swing2LimitDegrees=10, twistLimitDegrees=5}
 {bone1Name=dyn_belt_bk_05, bone2Name=dyn_belt_bk_04, swing1LimitDegrees=10, swing2LimitDegrees=10, twistLimitDegrees=5}
 {bone1Name=dyn_belt_bk_06, bone2Name=dyn_belt_bk_05, swing1LimitDegrees=10, swing2LimitDegrees=10, twistLimitDegrees=5}
 {bone1Name=dyn_belt_bk_07, bone2Name=dyn_belt_bk_06, swing1LimitDegrees=10, swing2LimitDegrees=10, twistLimitDegrees=5}
 {bone1Name=dyn_belt_frnt_01, bone2Name=dyn_belt_master, swing1LimitDegrees=5, swing2LimitDegrees=5, twistLimitDegrees=0}
 {bone1Name=dyn_belt_frnt_02, bone2Name=dyn_belt_frnt_01, swing1LimitDegrees=10, swing2LimitDegrees=10, twistLimitDegrees=5}
 {bone1Name=dyn_belt_frnt_03, bone2Name=dyn_belt_frnt_02, swing1LimitDegrees=10, swing2LimitDegrees=10, twistLimitDegrees=5}
 {bone1Name=dyn_belt_frnt_04, bone2Name=dyn_belt_frnt_03, swing1LimitDegrees=10, swing2LimitDegrees=10, twistLimitDegrees=5}
 {bone1Name=dyn_belt_frnt_05, bone2Name=dyn_belt_frnt_04, swing1LimitDegrees=10, swing2LimitDegrees=10, twistLimitDegrees=5}
 {bone1Name=dyn_belt_frnt_06, bone2Name=dyn_belt_frnt_05, swing1LimitDegrees=10, swing2LimitDegrees=10, twistLimitDegrees=15}
 {bone1Name=dyn_belt_frnt_07, bone2Name=dyn_belt_frnt_06, swing1LimitDegrees=10, swing2LimitDegrees=10, twistLimitDegrees=15}
 {
  bone1Name=spine_01
  bone2Name=pelvis
  swing1LimitDegrees=45
  swing1Motion=Free
  swing2LimitDegrees=45
  swing2Motion=Free
  twistLimitDegrees=45
  twistMotion=Free
 }
 {
  bone1Name=spine_02
  bone2Name=spine_01
  swing1LimitDegrees=45
  swing1Motion=Free
  swing2LimitDegrees=45
  swing2Motion=Free
  twistLimitDegrees=45
  twistMotion=Free
 }
 {
  bone1Name=thigh_l
  bone2Name=pelvis
  swing1LimitDegrees=45
  swing1Motion=Free
  swing2LimitDegrees=45
  swing2Motion=Free
  twistLimitDegrees=45
  twistMotion=Free
 }
 {
  bone1Name=thigh_r
  bone2Name=pelvis
  swing1LimitDegrees=45
  swing1Motion=Free
  swing2LimitDegrees=45
  swing2Motion=Free
  twistLimitDegrees=45
  twistMotion=Free
 }
 {
  bone1Name=dyn_belt_anchor_base
  bone2Name=pelvis
  swing1LimitDegrees=45
  swing1Motion=Free
  swing2LimitDegrees=45
  swing2Motion=Free
  twistLimitDegrees=45
  twistMotion=Free
 }
 {bone1Name=dyn_belt_anchor_01, bone2Name=dyn_belt_anchor_base, swing1LimitDegrees=5, swing2LimitDegrees=5, twistLimitDegrees=10}
 {bone1Name=dyn_belt_anchor_02, bone2Name=dyn_belt_anchor_01, swing1LimitDegrees=5, swing2LimitDegrees=5, twistLimitDegrees=5}
 {bone1Name=dyn_belt_bk_07, bone2Name=dyn_belt_frnt_07, swing1LimitDegrees=10, swing2LimitDegrees=10, twistLimitDegrees=15}
 {bone1Name=dyn_belt_frnt_07, bone2Name=dyn_belt_anchor_02, swing1LimitDegrees=10, swing2LimitDegrees=10, twistLimitDegrees=15}
 {bone1Name=dyn_belt_bk_07, bone2Name=dyn_belt_anchor_02, swing1LimitDegrees=10, swing2LimitDegrees=10, twistLimitDegrees=5}
 {
  bone1Name=dyn_greenBelt_anchor_base
  bone2Name=pelvis
  swing1LimitDegrees=45
  swing1Motion=Free
  swing2LimitDegrees=45
  swing2Motion=Free
  twistLimitDegrees=45
  twistMotion=Free
 }
 {bone1Name=dyn_greenBelt_anchor_01, bone2Name=dyn_greenBelt_anchor_base, swing1LimitDegrees=5, swing2LimitDegrees=5, twistLimitDegrees=5}
 {bone1Name=dyn_greenBelt_anchor_02, bone2Name=dyn_greenBelt_anchor_01, swing1LimitDegrees=5, swing2LimitDegrees=5, twistLimitDegrees=5}
 {
  bone1Name=dyn_greenBelt_master
  bone2Name=pelvis
  swing1LimitDegrees=45
  swing1Motion=Free
  swing2LimitDegrees=45
  swing2Motion=Free
  twistLimitDegrees=45
  twistMotion=Free
 }
 {
  bone1Name=dyn_greenBelt_bk_01
  bone2Name=dyn_greenBelt_master
  swing1LimitDegrees=45
  swing1Motion=Free
  swing2LimitDegrees=45
  swing2Motion=Free
  twistLimitDegrees=45
  twistMotion=Free
 }
 {
  bone1Name=dyn_greenBelt_bk_02
  bone2Name=dyn_greenBelt_bk_01
  swing1LimitDegrees=45
  swing1Motion=Free
  swing2LimitDegrees=45
  swing2Motion=Free
  twistLimitDegrees=45
  twistMotion=Free
 }
 {
  bone1Name=dyn_greenBelt_bk_03
  bone2Name=dyn_greenBelt_bk_02
  swing1LimitDegrees=45
  swing1Motion=Free
  swing2LimitDegrees=45
  swing2Motion=Free
  twistLimitDegrees=45
  twistMotion=Free
 }
 {
  bone1Name=dyn_greenBelt_bk_04
  bone2Name=dyn_greenBelt_bk_03
  swing1LimitDegrees=45
  swing1Motion=Free
  swing2LimitDegrees=45
  swing2Motion=Free
  twistLimitDegrees=45
  twistMotion=Free
 }
 {bone1Name=dyn_greenBelt_bk_05, bone2Name=dyn_greenBelt_bk_04, swing1LimitDegrees=10, swing2LimitDegrees=10, twistLimitDegrees=15}
 {bone1Name=dyn_greenBelt_bk_06, bone2Name=dyn_greenBelt_bk_05, swing1LimitDegrees=10, swing2LimitDegrees=10, twistLimitDegrees=15}
 {bone1Name=dyn_greenBelt_bk_07, bone2Name=dyn_greenBelt_bk_06, swing1LimitDegrees=10, swing2LimitDegrees=10, twistLimitDegrees=15}
 {
  bone1Name=dyn_greenBelt_frnt_01
  bone2Name=dyn_greenBelt_master
  swing1LimitDegrees=45
  swing1Motion=Free
  swing2LimitDegrees=45
  swing2Motion=Free
  twistLimitDegrees=45
  twistMotion=Free
 }
 {
  bone1Name=dyn_greenBelt_frnt_02
  bone2Name=dyn_greenBelt_frnt_01
  swing1LimitDegrees=45
  swing1Motion=Free
  swing2LimitDegrees=45
  swing2Motion=Free
  twistLimitDegrees=45
  twistMotion=Free
 }
 {bone1Name=dyn_greenBelt_frnt_03, bone2Name=dyn_greenBelt_frnt_02, swing1LimitDegrees=10, swing2LimitDegrees=10, twistLimitDegrees=15}
 {bone1Name=dyn_greenBelt_frnt_04, bone2Name=dyn_greenBelt_frnt_03, swing1LimitDegrees=10, swing2LimitDegrees=10, twistLimitDegrees=15}
 {bone1Name=dyn_greenBelt_frnt_05, bone2Name=dyn_greenBelt_frnt_04, swing1LimitDegrees=10, swing2LimitDegrees=10, twistLimitDegrees=15}
 {bone1Name=dyn_greenBelt_frnt_06, bone2Name=dyn_greenBelt_frnt_05, swing1LimitDegrees=10, swing2LimitDegrees=10, twistLimitDegrees=15}
 {bone1Name=dyn_greenBelt_frnt_06, bone2Name=dyn_greenBelt_bk_07, swing1LimitDegrees=45, swing2LimitDegrees=45, twistLimitDegrees=15}
 {bone1Name=dyn_greenBelt_frnt_06, bone2Name=dyn_greenBelt_anchor_02, swing1LimitDegrees=45, swing2LimitDegrees=45, twistLimitDegrees=15}
 {
  bone1Name=dyn_greenBelt_bk_07
  bone2Name=dyn_greenBelt_anchor_02
  swing1LimitDegrees=45
  swing1Motion=Free
  swing2LimitDegrees=45
  swing2Motion=Free
  twistLimitDegrees=45
  twistMotion=Free
 }
 {bone1Name=dyn_belt_frnt_03, bone2Name=dyn_greenBelt_frnt_03, swing1LimitDegrees=10, swing2LimitDegrees=10, twistLimitDegrees=15}
 {
  bone1Name=dyn_belt_mask_base
  bone2Name=dyn_belt_bk_02
  swing1LimitDegrees=45
  swing1Motion=Free
  swing2LimitDegrees=45
  swing2Motion=Free
  twistLimitDegrees=45
  twistMotion=Free
 }
 {bone1Name=dyn_belt_mask_01, bone2Name=dyn_belt_mask_base, swing1LimitDegrees=35, swing2LimitDegrees=30, twistLimitDegrees=10}
 {
  bone1Name=calf_l
  bone2Name=thigh_l
  swing1LimitDegrees=45
  swing1Motion=Free
  swing2LimitDegrees=45
  swing2Motion=Free
  twistLimitDegrees=45
  twistMotion=Free
 }
 {
  bone1Name=foot_l
  bone2Name=calf_l
  swing1LimitDegrees=45
  swing1Motion=Free
  swing2LimitDegrees=45
  swing2Motion=Free
  twistLimitDegrees=45
  twistMotion=Free
 }
 {
  bone1Name=dyn_belt_frnt_01
  bone2Name=dyn_greenBelt_frnt_06
  swing1LimitDegrees=45
  swing1Motion=Locked
  swing2LimitDegrees=45
  swing2Motion=Locked
  twistLimitDegrees=45
  twistMotion=Locked
 }
]
