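# Physics asset for the SmoothSuede medium female body's dynamic jacket pieces (named Straps), with capsule bodies for jacket bones, spine, pelvis and thighs.
# Five two-bone jacket chains (front left/right; back left, middle, right) branch from spine_04, with torso and thigh capsules providing body volume.
# Base and body joints swing and twist freely, while jacket tip joints have angular limits, tightest on the back, which allows only slight swing on one axis.
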PhysicsAsset /BRCosmetics/Characters/Player/Female/Medium/Bodies/F_MED_SmoothSuede/Meshes/F_MED_SmoothSuede_Physics_Straps
bodies {
 dyn_jacket_bk_l_01 [{shapeType=Capsule, length=6.6625, radius=3.4479, shapeName=dyn_jacket_bk_l_01_capsule, center=<7.7324 2.5844 0>, rotation=<rot 90 0 0>}]
 dyn_jacket_bk_l_base [{shapeType=Capsule, length=0.505, radius=0.505, shapeName=dyn_jacket_bk_l_base_capsule, rotation=<rot 0 -90 0>}]
 dyn_jacket_bk_mid_01 [{shapeType=Capsule, length=6.5112, radius=3.735, shapeName=dyn_jacket_bk_mid_01_capsule, center=<7.3115 3.0568 0>, rotation=<rot 90 0 0>}]
 dyn_jacket_bk_mid_base [{shapeType=Capsule, length=0.505, radius=0.505, shapeName=dyn_jacket_bk_mid_base_capsule, rotation=<rot 0 -90 0>}]
 dyn_jacket_bk_r_01 [{shapeType=Capsule, length=8.2646, radius=3.6101, shapeName=dyn_jacket_bk_r_01_capsule, center=<8.0842 1.7923 0>, rotation=<rot 90 0 0>}]
 dyn_jacket_bk_r_base [{shapeType=Capsule, length=0.505, radius=0.505, shapeName=dyn_jacket_bk_r_base_capsule, rotation=<rot 0 -90 0>}]
 dyn_jacket_frnt_l_01 [{
   shapeType=Capsule
   length=8.4222
   radius=4.2322
   shapeName=dyn_jacket_frnt_l_01_capsule
   center=<7.8978 -0.0447 0>
   rotation=<rot 90 180 -177.5574>
  }]
 dyn_jacket_frnt_l_base [{shapeType=Capsule, length=4.319, radius=2.5029, shapeName=dyn_jacket_frnt_l_base_capsule}]
 dyn_jacket_frnt_r_01 [{
   shapeType=Capsule
   length=7.208
   radius=3.5811
   shapeName=dyn_jacket_frnt_r_01_capsule
   center=<7.3459 -0.5185 0>
   rotation=<rot 90 180 0.7638>
  }]
 dyn_jacket_frnt_r_base [{shapeType=Capsule, length=0.505, radius=0.505, shapeName=dyn_jacket_frnt_r_base_capsule}]
 pelvis [{shapeType=Capsule, length=8.3392, radius=15.1196, shapeName=pelvis_capsule, center=<-2.5422 -1.1583 0>}]
 spine_01 [{shapeType=Capsule, length=9.6511, radius=11.3016, shapeName=spine_01_capsule, center=<-0.1596 0.4225 0>}]
 spine_02 [{shapeType=Capsule, length=16.8025, radius=9.7549, shapeName=spine_02_capsule, center=<0 0.2063 0>}]
 spine_03 [{shapeType=Capsule, length=0.505, radius=0.505, shapeName=spine_03_capsule}]
 spine_04 [{shapeType=Capsule, length=7.753, radius=11.0617, shapeName=spine_04_capsule}]
 thigh_l [{shapeType=Capsule, length=31.3236, radius=10.4134, shapeName=thigh_l_capsule, center=<-19.0307 0 -0.6365>, rotation=<rot 90 0 0>}]
 thigh_r [{shapeType=Capsule, length=32.1053, radius=9.6318, shapeName=thigh_r_capsule, center=<19.0307 0.1897 0.6365>, rotation=<rot 90 0 180>}]
}
constraints [
 {
  bone1Name=spine_04
  bone2Name=spine_03
  swing1LimitDegrees=45
  swing1Motion=Free
  swing2LimitDegrees=45
  swing2Motion=Free
  twistLimitDegrees=45
  twistMotion=Free
 }
 {
  bone1Name=spine_03
  bone2Name=spine_02
  swing1LimitDegrees=45
  swing1Motion=Free
  swing2LimitDegrees=45
  swing2Motion=Free
  twistLimitDegrees=45
  twistMotion=Free
 }
 {
  bone1Name=spine_02
  bone2Name=spine_01
  swing1LimitDegrees=45
  swing1Motion=Free
  swing2LimitDegrees=45
  swing2Motion=Free
  twistLimitDegrees=45
  twistMotion=Free
 }
 {
  bone1Name=spine_01
  bone2Name=pelvis
  swing1LimitDegrees=45
  swing1Motion=Free
  swing2LimitDegrees=45
  swing2Motion=Free
  twistLimitDegrees=45
  twistMotion=Free
 }
 {
  bone1Name=dyn_jacket_frnt_l_base
  bone2Name=spine_04
  swing1LimitDegrees=45
  swing1Motion=Free
  swing2LimitDegrees=45
  swing2Motion=Free
  twistLimitDegrees=45
  twistMotion=Free
 }
 {bone1Name=dyn_jacket_frnt_l_01, bone2Name=dyn_jacket_frnt_l_base, swing1LimitDegrees=10, swing2LimitDegrees=30, twistLimitDegrees=5}
 {
  bone1Name=dyn_jacket_frnt_r_base
  bone2Name=spine_04
  swing1LimitDegrees=45
  swing1Motion=Free
  swing2LimitDegrees=45
  swing2Motion=Free
  twistLimitDegrees=45
  twistMotion=Free
 }
 {bone1Name=dyn_jacket_frnt_r_01, bone2Name=dyn_jacket_frnt_r_base, swing1LimitDegrees=30, swing2LimitDegrees=45, twistLimitDegrees=15}
 {
  bone1Name=thigh_l
  bone2Name=pelvis
  swing1LimitDegrees=45
  swing1Motion=Free
  swing2LimitDegrees=45
  swing2Motion=Free
  twistLimitDegrees=45
  twistMotion=Free
 }
 {
  bone1Name=thigh_r
  bone2Name=pelvis
  swing1LimitDegrees=45
  swing1Motion=Free
  swing2LimitDegrees=45
  swing2Motion=Free
  twistLimitDegrees=45
  twistMotion=Free
 }
 {
  bone1Name=dyn_jacket_bk_l_base
  bone2Name=spine_04
  swing1LimitDegrees=45
  swing1Motion=Free
  swing2LimitDegrees=45
  swing2Motion=Free
  twistLimitDegrees=45
  twistMotion=Free
 }
 {bone1Name=dyn_jacket_bk_l_01, bone2Name=dyn_jacket_bk_l_base, swing1LimitDegrees=10, swing2LimitDegrees=0, twistLimitDegrees=0}
 {
  bone1Name=dyn_jacket_bk_mid_base
  bone2Name=spine_04
  swing1LimitDegrees=45
  swing1Motion=Free
  swing2LimitDegrees=45
  swing2Motion=Free
  twistLimitDegrees=45
  twistMotion=Free
 }
 {bone1Name=dyn_jacket_bk_mid_01, bone2Name=dyn_jacket_bk_mid_base, swing1LimitDegrees=10, swing2LimitDegrees=0, twistLimitDegrees=0}
 {
  bone1Name=dyn_jacket_bk_r_base
  bone2Name=spine_04
  swing1LimitDegrees=45
  swing1Motion=Free
  swing2LimitDegrees=45
  swing2Motion=Free
  twistLimitDegrees=45
  twistMotion=Free
 }
 {bone1Name=dyn_jacket_bk_r_01, bone2Name=dyn_jacket_bk_r_base, swing1LimitDegrees=10, swing2LimitDegrees=0, twistLimitDegrees=0}
]
