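# Animation Blueprint for the SmoothSuede female face accessory mesh that adds secondary motion to dreadlocks and a left earring.
# It copies the attached parent mesh's pose, then constrains the earring bone to the head.
# Seven chained Anim Dynamics nodes simulate front, side, back and top dreads with springs, gravity and angular limits in root-relative space.
# That result is cached as 'ToEarring', feeding a final, softer earring simulation before output; the EventGraph holds no logic.
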AnimBlueprint /BRCosmetics/Characters/Player/Female/Medium/Bodies/F_MED_SmoothSuede/Meshes/Parts/F_MED_SmoothSuede_FaceAcc_AnimBP
target_skeleton=&/F_MED_SmoothSuede_FaceAcc_Skeleton
graphs [
 AnimGraph {nodes [
   n1 = Anim Dynamics - Bone: dyn_earring_l(ComponentPose=Local To Component(LocalPose=Use cached pose 'ToEarring'()), Alpha=1, ExternalForce="0,0,0") {
    angularDampingOverride=1
    angularSpringConstant=20
    bAngularSpring=true
    bOverrideAngularDamping=true
    bUseGravityOverride=true
    simulationSpace=RootRelative
    alphaBoolBlend {blendOption=Linear}
    boundBone {boneName=dyn_earring_l}
    componentLinearAccScale=<1 1 1>
    gravityOverride {z=-980}
    physicsBodyDefinitions [{
      boundBone {boneName=dyn_earring_l}
      boxExtents=<2 2 2>
      constraintSetup {angularLimitsMax=<10 15 20>, angularLimitsMin=<-10 -25 -15>, angularTarget=<0 0 0>}
      localJointOffset=<0.1 0 0>
     }]
   }
   Output Pose(Result=Component To Local(ComponentPose=@n1)) {layerGroup=None}
   n2 = "Constraint - dyn_earring_l (P)"!(
    ComponentPose=Local To Component(LocalPose=Copy Pose From Mesh() {bUseAttachedParent=true})
    Alpha=1
    ConstraintWeights_0=1
    ConstraintWeights_1:float=0
   ) -> (Pose) {
    alphaBoolBlend {blendOption=Linear}
    boneToModify {boneName=dyn_earring_l}
    constraintSetup [{targetBone {boneName=head}}]
    constraintWeights [1,0]
   }
   n3 = Anim Dynamics - Bone: dyn_dread_frnt_l_01(ComponentPose=@n2, Alpha=1, ExternalForce="0,0,0") {
    angularDampingOverride=1
    angularSpringConstant=800
    bAngularSpring=true
    bOverrideAngularDamping=true
    bUseGravityOverride=true
    simulationSpace=RootRelative
    alphaBoolBlend {blendOption=Linear}
    boundBone {boneName=dyn_dread_frnt_l_01}
    componentAppliedLinearAccClamp=<2000 2000 2000>
    componentLinearAccScale=<1 1 1>
    gravityOverride {z=-980}
    physicsBodyDefinitions [{
      boundBone {boneName=dyn_dread_frnt_l_01}
      boxExtents=<10 10 10>
      constraintSetup {angularLimitsMax=<20 0 40>, angularLimitsMin=<-20 -30 -20>, angularTarget=<0 0 0>}
      localJointOffset=<0.6 0 0>
     }]
   }
   n4 = Anim Dynamics - Bone: dyn_dread_frnt_r_01(ComponentPose=@n3, Alpha=1, ExternalForce="0,0,0") {
    angularDampingOverride=1
    angularSpringConstant=800
    bAngularSpring=true
    bOverrideAngularDamping=true
    bUseGravityOverride=true
    simulationSpace=RootRelative
    alphaBoolBlend {blendOption=Linear}
    boundBone {boneName=dyn_dread_frnt_r_01}
    componentAppliedLinearAccClamp=<2000 2000 2000>
    componentLinearAccScale=<1 1 1>
    gravityOverride {z=-980}
    physicsBodyDefinitions [{
      boundBone {boneName=dyn_dread_frnt_r_01}
      boxExtents=<10 10 10>
      constraintSetup {angularLimitsMax=<15 0 0>, angularLimitsMin=<-20 -30 -20>, angularTarget=<0 0 0>}
      localJointOffset=<0.6 0 0>
     }]
   }
   n5 = Anim Dynamics - Bone: dyn_dread_side_r_01(ComponentPose=@n4, Alpha=1, ExternalForce="0,0,0") {
    angularDampingOverride=0.999
    angularSpringConstant=800
    bAngularSpring=true
    bOverrideAngularDamping=true
    bUseGravityOverride=true
    simulationSpace=RootRelative
    alphaBoolBlend {blendOption=Linear}
    boundBone {boneName=dyn_dread_side_r_01}
    componentAppliedLinearAccClamp=<2000 2000 2000>
    componentLinearAccScale=<1 1 1>
    gravityOverride {z=-980}
    physicsBodyDefinitions [{
      boundBone {boneName=dyn_dread_side_r_01}
      boxExtents=<10 10 10>
      constraintSetup {angularLimitsMax=<10 10 5>, angularLimitsMin=<-10 -60 -20>, angularTarget=<0 0 0>}
      localJointOffset=<0.6 0 0>
     }]
   }
   n6 = Anim Dynamics - Bone: dyn_dread_bk_r_01(ComponentPose=@n5, Alpha=1, ExternalForce="0,0,0") {
    angularDampingOverride=0.999
    angularSpringConstant=800
    bAngularSpring=true
    bOverrideAngularDamping=true
    bUseGravityOverride=true
    simulationSpace=RootRelative
    alphaBoolBlend {blendOption=Linear}
    boundBone {boneName=dyn_dread_bk_r_01}
    componentAppliedLinearAccClamp=<2000 2000 2000>
    componentLinearAccScale=<1 1 1>
    gravityOverride {z=-980}
    physicsBodyDefinitions [{
      boundBone {boneName=dyn_dread_bk_r_01}
      boxExtents=<10 10 10>
      constraintSetup {angularLimitsMax=<10 0 10>, angularLimitsMin=<-10 -50 -10>, angularTarget=<0 0 0>}
      localJointOffset=<0.6 0 0>
     }]
   }
   n7 = Anim Dynamics - Bone: dyn_dread_top_r_01(ComponentPose=@n6, Alpha=1, ExternalForce="0,0,0") {
    angularDampingOverride=1
    angularSpringConstant=800
    bAngularSpring=true
    bOverrideAngularDamping=true
    bUseGravityOverride=true
    simulationSpace=RootRelative
    alphaBoolBlend {blendOption=Linear}
    boundBone {boneName=dyn_dread_top_r_01}
    componentAppliedLinearAccClamp=<2000 2000 2000>
    componentLinearAccScale=<1 1 1>
    gravityOverride {z=-980}
    physicsBodyDefinitions [{
      boundBone {boneName=dyn_dread_top_r_01}
      boxExtents=<10 10 10>
      constraintSetup {angularLimitsMax=<10 0 10>, angularLimitsMin=<-10 -30 -10>, angularTarget=<0 0 0>}
      localJointOffset=<0.6 0 0>
     }]
   }
   n8 = Anim Dynamics - Bone: dyn_dread_top_mid_01(ComponentPose=@n7, Alpha=1, ExternalForce="0,0,0") {
    angularDampingOverride=1
    angularSpringConstant=800
    bAngularSpring=true
    bOverrideAngularDamping=true
    bUseGravityOverride=true
    simulationSpace=RootRelative
    alphaBoolBlend {blendOption=Linear}
    boundBone {boneName=dyn_dread_top_mid_01}
    componentAppliedLinearAccClamp=<2000 2000 2000>
    componentLinearAccScale=<1 1 1>
    gravityOverride {z=-980}
    physicsBodyDefinitions [{
      boundBone {boneName=dyn_dread_top_mid_01}
      boxExtents=<10 10 10>
      constraintSetup {angularLimitsMax=<10 0 10>, angularLimitsMin=<-10 -30 -10>, angularTarget=<0 0 0>}
      localJointOffset=<0.6 0 0>
     }]
   }
   n9 = Anim Dynamics - Bone: dyn_dread_top_bk_01(ComponentPose=@n8, Alpha=1, ExternalForce="0,0,0") {
    angularDampingOverride=1
    angularSpringConstant=800
    bAngularSpring=true
    bOverrideAngularDamping=true
    bUseGravityOverride=true
    simulationSpace=RootRelative
    alphaBoolBlend {blendOption=Linear}
    boundBone {boneName=dyn_dread_top_bk_01}
    componentAppliedLinearAccClamp=<2000 2000 2000>
    componentLinearAccScale=<1 1 1>
    gravityOverride {z=-980}
    physicsBodyDefinitions [{
      boundBone {boneName=dyn_dread_top_bk_01}
      boxExtents=<10 10 10>
      constraintSetup {angularLimitsMax=<10 0 10>, angularLimitsMin=<-10 -30 -10>, angularTarget=<0 0 0>}
      localJointOffset=<0.6 0 0>
     }]
   }
   Save cached pose 'ToEarring'(Pose=Component To Local(ComponentPose=@n9)) {cacheName=ToEarring}
  ]}
 EventGraph {nodes [Event BlueprintUpdateAnimation()]}
]
variables []
